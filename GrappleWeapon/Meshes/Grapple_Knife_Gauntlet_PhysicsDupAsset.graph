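# Physics asset for the grapple knife gauntlet mesh, defining capsule bodies and joints for two rope bone chains.
# Each chain runs from a hand through nine rope bodies with free swing, and the two tips are constrained together, linking the hands.
# Hands joint to forearms with free swing and twist; other arm and spine bodies are tiny unconstrained capsules.
# The pelvis carries a large, offset, zero-length capsule, effectively a sphere.
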PhysicsAsset /GrappleWeapon/Meshes/Grapple_Knife_Gauntlet_PhysicsDupAsset
bodies {
 clavicle_l [{shapeType=Capsule, length=1, radius=1, shapeName=clavicle_l_capsule}]
 clavicle_r [{shapeType=Capsule, length=1, radius=1, shapeName=clavicle_r_capsule}]
 hand_l [{shapeType=Capsule, length=1, radius=1, shapeName=hand_l_capsule}]
 hand_r [{shapeType=Capsule, length=1, radius=1, shapeName=hand_r_capsule}]
 lowerarm_l [{shapeType=Capsule, length=1, radius=1, shapeName=lowerarm_l_capsule}]
 lowerarm_r [{shapeType=Capsule, length=1, radius=1, shapeName=lowerarm_r_capsule}]
 pelvis [{shapeType=Capsule, length=0, radius=100, center=<-136.1542 -21.738 0>, rotation=<rot 90 0 0>}]
 rope_00_l [{shapeType=Capsule, length=2, radius=5, shapeName=rope_02_l_capsule, center=<2 0 0>, rotation=<rot 90 0 0>}]
 rope_00_r [{shapeType=Capsule, length=2, radius=5, shapeName=rope_02_r_capsule, center=<2 0 0>, rotation=<rot 90 0 0>}]
 rope_01_l [{shapeType=Capsule, length=2, radius=5, shapeName=rope_02_l_capsule, center=<2 0 0>, rotation=<rot 90 0 0>}]
 rope_01_r [{shapeType=Capsule, length=2, radius=5, shapeName=rope_02_r_capsule, center=<2 0 0>, rotation=<rot 90 0 0>}]
 rope_02_l [{shapeType=Capsule, length=2, radius=5, shapeName=rope_02_l_capsule, center=<2 0 0>, rotation=<rot 90 0 0>}]
 rope_02_r [{shapeType=Capsule, length=2, radius=5, shapeName=rope_02_r_capsule, center=<2 0 0>, rotation=<rot 90 0 0>}]
 rope_03_l [{shapeType=Capsule, length=2, radius=5, shapeName=rope_03_l_capsule, center=<2 0 0>, rotation=<rot 90 0 0>}]
 rope_03_r [{shapeType=Capsule, length=2, radius=5, shapeName=rope_03_r_capsule, center=<2 0 0>, rotation=<rot 90 0 0>}]
 rope_04_l [{shapeType=Capsule, length=2, radius=5, shapeName=rope_04_l_capsule, center=<2 0 0>, rotation=<rot 90 0 0>}]
 rope_04_r [{shapeType=Capsule, length=2, radius=5, shapeName=rope_04_r_capsule, center=<2 0 0>, rotation=<rot 90 0 0>}]
 rope_05_l [{shapeType=Capsule, length=2, radius=5, shapeName=rope_05_l_capsule, center=<2 0 0>, rotation=<rot 90 0 0>}]
 rope_05_r [{shapeType=Capsule, length=2, radius=5, shapeName=rope_05_r_capsule, center=<2 0 0>, rotation=<rot 90 0 0>}]
 rope_06_l [{shapeType=Capsule, length=2, radius=5, shapeName=rope_06_l_capsule, center=<2 0 0>, rotation=<rot 90 0 0>}]
 rope_06_r [{shapeType=Capsule, length=2, radius=5, shapeName=rope_06_r_capsule, center=<2 0 0>, rotation=<rot 90 0 0>}]
 rope_07_l [{shapeType=Capsule, length=2, radius=5, shapeName=rope_07_l_capsule, center=<2 0 0>, rotation=<rot 90 0 0>}]
 rope_07_r [{shapeType=Capsule, length=2, radius=5, shapeName=rope_07_r_capsule, center=<2 0 0>, rotation=<rot 90 0 0>}]
 rope_08_l [{shapeType=Capsule, length=2, radius=5, shapeName=rope_08_l_capsule}]
 rope_08_r [{shapeType=Capsule, length=2, radius=5, shapeName=rope_08_r_capsule}]
 spine_01 [{shapeType=Capsule, length=1, radius=1, shapeName=spine_01_capsule}]
 spine_02 [{shapeType=Capsule, length=1, radius=1, shapeName=spine_02_capsule}]
 spine_03 [{shapeType=Capsule, length=1, radius=1, shapeName=spine_03_capsule}]
 spine_04 [{shapeType=Capsule, length=1, radius=1, shapeName=spine_04_capsule}]
 spine_05 [{shapeType=Capsule, length=1, radius=1, shapeName=spine_05_capsule}]
 upperarm_l [{shapeType=Capsule, length=1, radius=1, shapeName=upperarm_l_capsule}]
 upperarm_r [{shapeType=Capsule, length=1, radius=1, shapeName=upperarm_r_capsule}]
}
constraints [
 {
  bone1Name=hand_l
  bone2Name=lowerarm_l
  swing1LimitDegrees=45
  swing1Motion=Free
  swing2LimitDegrees=45
  swing2Motion=Free
  twistLimitDegrees=45
  twistMotion=Free
 }
 {
  bone1Name=hand_r
  bone2Name=lowerarm_r
  swing1LimitDegrees=45
  swing1Motion=Free
  swing2LimitDegrees=45
  swing2Motion=Free
  twistLimitDegrees=45
  twistMotion=Free
 }
 {
  bone1Name=rope_02_l
  bone2Name=rope_01_l
  swing1LimitDegrees=45
  swing1Motion=Free
  swing2LimitDegrees=45
  swing2Motion=Free
  twistLimitDegrees=0
 }
 {
  bone1Name=rope_03_l
  bone2Name=rope_02_l
  swing1LimitDegrees=45
  swing1Motion=Free
  swing2LimitDegrees=45
  swing2Motion=Free
  twistLimitDegrees=0
 }
 {
  bone1Name=rope_04_l
  bone2Name=rope_03_l
  swing1LimitDegrees=45
  swing1Motion=Free
  swing2LimitDegrees=45
  swing2Motion=Free
  twistLimitDegrees=0
 }
 {
  bone1Name=rope_05_l
  bone2Name=rope_04_l
  swing1LimitDegrees=45
  swing1Motion=Free
  swing2LimitDegrees=45
  swing2Motion=Free
  twistLimitDegrees=0
 }
 {
  bone1Name=rope_06_l
  bone2Name=rope_05_l
  swing1LimitDegrees=15
  swing1Motion=Free
  swing2LimitDegrees=15
  swing2Motion=Free
  twistLimitDegrees=0
 }
 {
  bone1Name=rope_07_l
  bone2Name=rope_06_l
  swing1LimitDegrees=15
  swing1Motion=Free
  swing2LimitDegrees=15
  swing2Motion=Free
  twistLimitDegrees=0
 }
 {
  bone1Name=rope_08_l
  bone2Name=rope_07_l
  swing1LimitDegrees=15
  swing1Motion=Free
  swing2LimitDegrees=15
  swing2Motion=Free
  twistLimitDegrees=0
 }
 {
  bone1Name=rope_02_r
  bone2Name=rope_01_r
  swing1LimitDegrees=45
  swing1Motion=Free
  swing2LimitDegrees=45
  swing2Motion=Free
  twistLimitDegrees=0
 }
 {
  bone1Name=rope_03_r
  bone2Name=rope_02_r
  swing1LimitDegrees=45
  swing1Motion=Free
  swing2LimitDegrees=45
  swing2Motion=Free
  twistLimitDegrees=0
 }
 {
  bone1Name=rope_04_r
  bone2Name=rope_03_r
  swing1LimitDegrees=45
  swing1Motion=Free
  swing2LimitDegrees=45
  swing2Motion=Free
  twistLimitDegrees=0
 }
 {
  bone1Name=rope_05_r
  bone2Name=rope_04_r
  swing1LimitDegrees=45
  swing1Motion=Free
  swing2LimitDegrees=45
  swing2Motion=Free
  twistLimitDegrees=0
 }
 {
  bone1Name=rope_06_r
  bone2Name=rope_05_r
  swing1LimitDegrees=15
  swing1Motion=Free
  swing2LimitDegrees=15
  swing2Motion=Free
  twistLimitDegrees=0
 }
 {
  bone1Name=rope_07_r
  bone2Name=rope_06_r
  swing1LimitDegrees=15
  swing1Motion=Free
  swing2LimitDegrees=15
  swing2Motion=Free
  twistLimitDegrees=0
 }
 {
  bone1Name=rope_08_r
  bone2Name=rope_07_r
  swing1LimitDegrees=15
  swing1Motion=Free
  swing2LimitDegrees=15
  swing2Motion=Free
  twistLimitDegrees=0
 }
 {
  bone1Name=rope_01_r
  bone2Name=rope_00_r
  swing1LimitDegrees=45
  swing1Motion=Free
  swing2LimitDegrees=45
  swing2Motion=Free
  twistLimitDegrees=0
 }
 {
  bone1Name=rope_01_l
  bone2Name=rope_00_l
  swing1LimitDegrees=45
  swing1Motion=Free
  swing2LimitDegrees=45
  swing2Motion=Free
  twistLimitDegrees=0
 }
 {
  bone1Name=rope_08_l
  bone2Name=rope_08_r
  swing1LimitDegrees=45
  swing1Motion=Free
  swing2LimitDegrees=45
  swing2Motion=Free
  twistLimitDegrees=45
  twistMotion=Free
 }
 {
  bone1Name=rope_00_l
  bone2Name=hand_l
  swing1LimitDegrees=45
  swing1Motion=Free
  swing2LimitDegrees=45
  swing2Motion=Free
  twistLimitDegrees=0
 }
 {
  bone1Name=rope_00_r
  bone2Name=hand_r
  swing1LimitDegrees=45
  swing1Motion=Free
  swing2LimitDegrees=45
  swing2Motion=Free
  twistLimitDegrees=0
 }
]
