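# Animation Blueprint for the grapple knife gauntlet that follows its attached parent mesh and adds secondary motion.
# Copy Pose From Mesh feeds a Control Rig, three chained RigidBody simulations with separate physics assets (the first active when not crouching, one jump-specific), then additive rope-bone offsets.
# The update event caches jump/fall, crouch-move and crouch-sprint states.
# A second chain of right-rope offsets is not connected to the output.
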AnimBlueprint /GrappleWeapon/Meshes/Grapple_Knife_Gauntlet_AnimBP
target_skeleton=&/Grapple_Knife_Gauntlet_Skeleton
graphs [
 AnimGraph {nodes [
   n1 = "Transform (Modify) Bone - Bone: rope_08_r"!(ComponentPose, bAlphaBoolEnabled:bool=True, Translation="0,-10,0", Rotation="0,0,0", Scale="1,1,1") -> (Pose) {
    alphaInputType=Bool
    translationMode=BMM_Additive
    alphaBoolBlend {blendOption=Linear}
    boneToModify {boneName=rope_08_r}
    translation {y=-10}
   }
   n2 = "Transform (Modify) Bone - Bone: rope_07_r"!(ComponentPose=@n1, bAlphaBoolEnabled:bool=True, Translation="0,-10,0", Rotation="0,0,0", Scale="1,1,1") -> (Pose) {
    alphaInputType=Bool
    translationMode=BMM_Additive
    alphaBoolBlend {blendOption=Linear}
    boneToModify {boneName=rope_07_r}
    translation {y=-10}
   }
   n3 = "Transform (Modify) Bone - Bone: rope_06_r"!(ComponentPose=@n2, bAlphaBoolEnabled:bool=True, Translation="0,-10,0", Rotation="0,0,0", Scale="1,1,1") -> (Pose) {
    alphaInputType=Bool
    translationMode=BMM_Additive
    alphaBoolBlend {blendOption=Linear}
    boneToModify {boneName=rope_06_r}
    translation {y=-10}
   }
   n4 = "Transform (Modify) Bone - Bone: rope_05_r"!(ComponentPose=@n3, bAlphaBoolEnabled:bool=True, Translation="0,-10,0", Rotation="0,0,0", Scale="1,1,1") -> (Pose) {
    alphaInputType=Bool
    translationMode=BMM_Additive
    alphaBoolBlend {blendOption=Linear}
    boneToModify {boneName=rope_05_r}
    translation {y=-10}
   }
   n5 = "Transform (Modify) Bone - Bone: rope_04_r"!(ComponentPose=@n4, bAlphaBoolEnabled:bool=True, Translation="0,-5,0", Rotation="0,0,0", Scale="1,1,1") -> (Pose) {
    alphaInputType=Bool
    translationMode=BMM_Additive
    alphaBoolBlend {blendOption=Linear}
    boneToModify {boneName=rope_04_r}
    translation {y=-5}
   }
   n6 = "Transform (Modify) Bone - Bone: rope_03_r"!(ComponentPose=@n5, bAlphaBoolEnabled:bool=True, Translation="0,15,0", Rotation="0,0,0", Scale="1,1,1") -> (Pose) {
    alphaInputType=Bool
    translationMode=BMM_Additive
    alphaBoolBlend {blendOption=Linear}
    boneToModify {boneName=rope_03_r}
    translation {y=15}
   }
   n7 = "Transform (Modify) Bone - Bone: rope_02_r"!(ComponentPose=@n6, bAlphaBoolEnabled:bool=True, Translation="0,15,0", Rotation="0,0,0", Scale="1,1,1") -> (Pose) {
    alphaInputType=Bool
    translationMode=BMM_Additive
    alphaBoolBlend {blendOption=Linear}
    boneToModify {boneName=rope_02_r}
    translation {y=15}
   }
   n8 = "Transform (Modify) Bone - Bone: rope_01_r"!(ComponentPose=@n7, bAlphaBoolEnabled:bool=True, Translation="0,10,0", Rotation="0,0,0", Scale="1,1,1") -> (Pose) {
    alphaInputType=Bool
    translationMode=BMM_Additive
    alphaBoolBlend {blendOption=Linear}
    boneToModify {boneName=rope_01_r}
    translation {y=10}
   }
   n9 = "Transform (Modify) Bone - Bone: rope_08_r"!(ComponentPose=@n8, Alpha:float=1, Translation="0,0,0", Rotation="0,0,0", Scale="1,1,1") -> (Pose) {
    translationMode=BMM_Additive
    alphaBoolBlend {blendOption=Linear}
    alphaScaleBiasClamp {bInterpResult=true}
    boneToModify {boneName=rope_08_r}
   }
   n10 = "Transform (Modify) Bone - Bone: rope_07_r"!(ComponentPose=@n9, Alpha:float=1, Translation="0,0,0", Rotation="0,0,0", Scale="1,1,1") -> (Pose) {
    translationMode=BMM_Additive
    alphaBoolBlend {blendOption=Linear}
    alphaScaleBiasClamp {bInterpResult=true}
    boneToModify {boneName=rope_07_r}
   }
   n11 = "Transform (Modify) Bone - Bone: rope_06_r"!(ComponentPose=@n10, Alpha:float=1, Translation="0,0,5", Rotation="0,0,0", Scale="1,1,1") -> (Pose) {
    translationMode=BMM_Additive
    alphaBoolBlend {blendOption=Linear}
    alphaScaleBiasClamp {bInterpResult=true}
    boneToModify {boneName=rope_06_r}
    translation {z=5}
   }
   n12 = "Transform (Modify) Bone - Bone: rope_05_r"!(ComponentPose=@n11, Alpha:float=1, Translation="0,0,0", Rotation="0,0,0", Scale="1,1,1") -> (Pose) {
    translationMode=BMM_Additive
    alphaBoolBlend {blendOption=Linear}
    alphaScaleBiasClamp {bInterpResult=true}
    boneToModify {boneName=rope_05_r}
   }
   n13 = "Transform (Modify) Bone - Bone: rope_04_r"!(ComponentPose=@n12, Alpha:float=1, Translation="0,0,0", Rotation="0,0,0", Scale="1,1,1") -> (Pose) {
    translationMode=BMM_Additive
    alphaBoolBlend {blendOption=Linear}
    alphaScaleBiasClamp {bInterpResult=true}
    boneToModify {boneName=rope_04_r}
   }
   n14 = "Transform (Modify) Bone - Bone: rope_03_r"!(ComponentPose=@n13, Alpha:float=1, Translation="-3,0,0", Rotation="0,0,0", Scale="1,1,1") -> (Pose) {
    translationMode=BMM_Additive
    alphaBoolBlend {blendOption=Linear}
    alphaScaleBiasClamp {bInterpResult=true}
    boneToModify {boneName=rope_03_r}
    translation {x=-3}
   }
   n15 = "Transform (Modify) Bone - Bone: rope_02_r"!(ComponentPose=@n14, Alpha:float=1, Translation="-3,0,0", Rotation="0,0,0", Scale="1,1,1") -> (Pose) {
    translationMode=BMM_Additive
    alphaBoolBlend {blendOption=Linear}
    alphaScaleBiasClamp {bInterpResult=true}
    boneToModify {boneName=rope_02_r}
    translation {x=-3}
   }
   "Transform (Modify) Bone - Bone: rope_01_r"!(ComponentPose=@n15, Alpha:float=1, Translation="6,0,-5", Rotation="0,0,0", Scale="1,1,1") -> (Pose) {
    translationMode=BMM_Additive
    alphaBoolBlend {blendOption=Linear}
    alphaScaleBiasClamp {bInterpResult=true}
    boneToModify {boneName=rope_01_r}
    translation {x=6, z=-5}
   }
   n16 = Control Rig(Source=Copy Pose From Mesh() {bUseAttachedParent=true}) {
    alphaBoolBlend {blendOption=Linear}
    controlRigAssetReference {blueprintRigClass=&/Grapple_Knife_Gauntlet_CtrlRig.Grapple_Knife_Gauntlet_CtrlRig_C}
   }
   n17 = RigidBody!(ComponentPose=Local To Component(LocalPose=@n16), bAlphaBoolEnabled:bool=NOT Boolean(A=Get bIsCrouching()), ExternalForce) -> (Pose) {
    alphaInputType=Bool
    bOverrideWorldGravity=true
    overridePhysicsAsset=&/Grapple_Knife_Gauntlet_PhysicsAsset
    simulationSpace=BaseBoneSpace
    alphaBoolBlend {blendOption=Linear}
    baseBoneRef {boneName=root}
    componentAppliedLinearAccClamp=<1500 1500 1500>
    componentLinearAccScale=<0.15 0.15 0.2>
    componentLinearVelScale=<0.15 0.15 0.2>
    overrideWorldGravity {z=-980}
   }
   n18 = RigidBody!(ComponentPose=@n17, bAlphaBoolEnabled, ExternalForce, bTransferBoneVelocities:bool=False) -> (Pose) {
    alphaInputType=Bool
    bOverrideWorldGravity=true
    overridePhysicsAsset=&/Grapple_Knife_Gauntlet_PhysicsDupAsset
    simulationSpace=BaseBoneSpace
    alphaBoolBlend {blendOption=Linear}
    baseBoneRef {boneName=root}
    componentAppliedLinearAccClamp=<1500 1500 1500>
    componentLinearAccScale=<0.15 0.15 0.2>
    componentLinearVelScale=<0.15 0.15 0.2>
    overrideWorldGravity {z=-980}
   }
   n19 = RigidBody!(ComponentPose=@n18, bAlphaBoolEnabled, ExternalForce) -> (Pose) {
    alphaInputType=Bool
    bOverrideWorldGravity=true
    overridePhysicsAsset=&/Grapple_Knife_Gauntlet_Jump_PhysicsAsset
    simulationSpace=BaseBoneSpace
    alphaBoolBlend {blendInTime=0.1, blendOption=Linear, blendOutTime=0.1}
    baseBoneRef {boneName=root}
    componentAppliedLinearAccClamp=<1500 1500 1500>
    componentLinearAccScale=<0.15 0.15 0.2>
    componentLinearVelScale=<0.15 0.15 0.2>
    overrideWorldGravity {z=-980}
   }
   n20 = "Transform (Modify) Bone - Bone: rope_08_l"!(ComponentPose=@n19, bAlphaBoolEnabled:bool=True, Translation="0,3,0", Rotation="0,0,0", Scale="1,1,1") -> (Pose) {alphaInputType=Bool, translationMode=BMM_Additive, alphaBoolBlend {blendOption=Linear}, boneToModify {boneName=rope_08_l}, translation {y=3}}
   n21 = "Transform (Modify) Bone - Bone: rope_07_l"!(ComponentPose=@n20, bAlphaBoolEnabled:bool=True, Translation="0,3,0", Rotation="0,0,0", Scale="1,1,1") -> (Pose) {alphaInputType=Bool, translationMode=BMM_Additive, alphaBoolBlend {blendOption=Linear}, boneToModify {boneName=rope_07_l}, translation {y=3}}
   n22 = "Transform (Modify) Bone - Bone: rope_06_l"!(ComponentPose=@n21, bAlphaBoolEnabled:bool=True, Translation="0,3,0", Rotation="0,0,0", Scale="1,1,1") -> (Pose) {alphaInputType=Bool, translationMode=BMM_Additive, alphaBoolBlend {blendOption=Linear}, boneToModify {boneName=rope_06_l}, translation {y=3}}
   n23 = "Transform (Modify) Bone - Bone: rope_05_l"!(ComponentPose=@n22, bAlphaBoolEnabled:bool=True, Translation="0,3,0", Rotation="0,0,0", Scale="1,1,1") -> (Pose) {alphaInputType=Bool, translationMode=BMM_Additive, alphaBoolBlend {blendOption=Linear}, boneToModify {boneName=rope_05_l}, translation {y=3}}
   n24 = "Transform (Modify) Bone - Bone: rope_04_l"!(ComponentPose=@n23, bAlphaBoolEnabled:bool=True, Translation="0,3,0", Rotation="0,0,0", Scale="1,1,1") -> (Pose) {alphaInputType=Bool, translationMode=BMM_Additive, alphaBoolBlend {blendOption=Linear}, boneToModify {boneName=rope_04_l}, translation {y=3}}
   n25 = "Transform (Modify) Bone - Bone: rope_03_l"!(ComponentPose=@n24, bAlphaBoolEnabled:bool=True, Translation="0,3,0", Rotation="0,0,0", Scale="1,1,1") -> (Pose) {alphaInputType=Bool, translationMode=BMM_Additive, alphaBoolBlend {blendOption=Linear}, boneToModify {boneName=rope_03_l}, translation {y=3}}
   n26 = "Transform (Modify) Bone - Bone: rope_02_l"!(ComponentPose=@n25, bAlphaBoolEnabled:bool=True, Translation="0,3,0", Rotation="0,0,0", Scale="1,1,1") -> (Pose) {alphaInputType=Bool, translationMode=BMM_Additive, alphaBoolBlend {blendOption=Linear}, boneToModify {boneName=rope_02_l}, translation {y=3}}
   n27 = "Transform (Modify) Bone - Bone: rope_01_l"!(ComponentPose=@n26, bAlphaBoolEnabled:bool=True, Translation="0,3,0", Rotation="0,0,0", Scale="1,1,1") -> (Pose) {alphaInputType=Bool, translationMode=BMM_Additive, alphaBoolBlend {blendOption=Linear}, boneToModify {boneName=rope_01_l}, translation {y=3}}
   n28 = "Transform (Modify) Bone - Bone: rope_08_r"!(ComponentPose=@n27, bAlphaBoolEnabled:bool=True, Translation="0,3,0", Rotation="0,0,0", Scale="1,1,1") -> (Pose) {alphaInputType=Bool, translationMode=BMM_Additive, alphaBoolBlend {blendOption=Linear}, boneToModify {boneName=rope_08_r}, translation {y=3}}
   n29 = "Transform (Modify) Bone - Bone: rope_07_r"!(ComponentPose=@n28, bAlphaBoolEnabled:bool=True, Translation="0,3,0", Rotation="0,0,0", Scale="1,1,1") -> (Pose) {alphaInputType=Bool, translationMode=BMM_Additive, alphaBoolBlend {blendOption=Linear}, boneToModify {boneName=rope_07_r}, translation {y=3}}
   n30 = "Transform (Modify) Bone - Bone: rope_06_r"!(ComponentPose=@n29, bAlphaBoolEnabled:bool=True, Translation="0,3,0", Rotation="0,0,0", Scale="1,1,1") -> (Pose) {alphaInputType=Bool, translationMode=BMM_Additive, alphaBoolBlend {blendOption=Linear}, boneToModify {boneName=rope_06_r}, translation {y=3}}
   n31 = "Transform (Modify) Bone - Bone: rope_05_r"!(ComponentPose=@n30, bAlphaBoolEnabled:bool=True, Translation="0,3,0", Rotation="0,0,0", Scale="1,1,1") -> (Pose) {alphaInputType=Bool, translationMode=BMM_Additive, alphaBoolBlend {blendOption=Linear}, boneToModify {boneName=rope_05_r}, translation {y=3}}
   n32 = "Transform (Modify) Bone - Bone: rope_04_r"!(ComponentPose=@n31, bAlphaBoolEnabled:bool=True, Translation="0,3,0", Rotation="0,0,0", Scale="1,1,1") -> (Pose) {alphaInputType=Bool, translationMode=BMM_Additive, alphaBoolBlend {blendOption=Linear}, boneToModify {boneName=rope_04_r}, translation {y=3}}
   n33 = "Transform (Modify) Bone - Bone: rope_03_r"!(ComponentPose=@n32, bAlphaBoolEnabled:bool=True, Translation="0,3,0", Rotation="0,0,0", Scale="1,1,1") -> (Pose) {alphaInputType=Bool, translationMode=BMM_Additive, alphaBoolBlend {blendOption=Linear}, boneToModify {boneName=rope_03_r}, translation {y=3}}
   n34 = "Transform (Modify) Bone - Bone: rope_02_r"!(ComponentPose=@n33, bAlphaBoolEnabled:bool=True, Translation="0,3,0", Rotation="0,0,0", Scale="1,1,1") -> (Pose) {alphaInputType=Bool, translationMode=BMM_Additive, alphaBoolBlend {blendOption=Linear}, boneToModify {boneName=rope_02_r}, translation {y=3}}
   n35 = "Transform (Modify) Bone - Bone: rope_01_r"!(ComponentPose=@n34, bAlphaBoolEnabled:bool=True, Translation="0,3,0", Rotation="0,0,0", Scale="1,1,1") -> (Pose) {alphaInputType=Bool, translationMode=BMM_Additive, alphaBoolBlend {blendOption=Linear}, boneToModify {boneName=rope_01_r}, translation {y=3}}
   Output Pose(Result=Component To Local(ComponentPose=@n35)) {layerGroup=None}
   Get bIsInVehicle() -> (bIsInVehicle:bool)
  ]}
 EventGraph {nodes [
   Event BlueprintUpdateAnimation()
   > Set isJumpingFalling(isJumpingFalling:bool=OR Boolean(A=Get bIsJumping(), B=Get bIsFalling())) -> (Output_Get:bool)
   > Set isCrouchMoving(isCrouchMoving:bool=AND Boolean(A=Get bIsMoving2D(), B=Get bIsCrouching())) -> (Output_Get:bool)
   > Set Is Crouch Sprinting(Is Crouch Sprinting:double="To Float (Boolean)"(InBool=Get bIsCrouchSprinting())) -> (Output_Get:double)
   TryGetPawnOwner()
  ]}
]
variables [
 Is Crouch Sprinting {type=number}
 isCrouchMoving {type=boolean}
 isJumpingFalling {type=boolean}
]
