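# AnimBlueprint for the grapple knife weapon skeleton, playing idle and air, full-body and upper-body attack animations.
# A conduit-hubbed state machine selects states from MantisAttack booleans, chains paired attacks with inertialization, and outputs through Dead Blending.
# A 24-node Anim Dynamics rope-bone chain exists but is not connected to the output pose.
# Mantis technique events set IsAirAttack, which no transition reads.
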AnimBlueprint /GrappleWeapon/Meshes/Grapple_Knife_Sharp_Skeleton_AnimBP
target_skeleton=&/Grapple_Knife_Sharp_Skeleton
graphs [
 AnimGraph {nodes [
   n1 = Anim Dynamics - Bone: rope_01(Alpha=1, ExternalForce="0,0,0") {
    angularDampingOverride=0.9999
    angularSpringConstant=20000
    bOverrideAngularDamping=true
    bUseGravityOverride=true
    simulationSpace=RootRelative
    alphaBoolBlend {blendOption=Linear}
    boundBone {boneName=rope_01}
    componentAppliedLinearAccClamp=<1500 1500 1500>
    componentLinearAccScale=<0.15 0.15 0.15>
    componentLinearVelScale=<0.15 0.15 0.15>
    gravityOverride {z=-980}
    physicsBodyDefinitions [{
      sphereCollisionRadius=0
      boundBone {boneName=rope_01}
      boxExtents=<10 15 15>
      constraintSetup {angularLimitsMax=<7 3 7>, angularLimitsMin=<-7 -3 -7>, angularTarget=<1 0 0>}
      localJointOffset=<4 0 0>
     }]
   }
   n2 = Anim Dynamics - Bone: rope_02(ComponentPose=@n1, Alpha=1, ExternalForce="0,0,0") {
    angularDampingOverride=0.9999
    angularSpringConstant=20000
    bOverrideAngularDamping=true
    bUseGravityOverride=true
    simulationSpace=RootRelative
    alphaBoolBlend {blendOption=Linear}
    boundBone {boneName=rope_02}
    componentAppliedLinearAccClamp=<1500 1500 1500>
    componentLinearAccScale=<0.15 0.15 0.15>
    componentLinearVelScale=<0.15 0.15 0.15>
    gravityOverride {z=-980}
    physicsBodyDefinitions [{
      sphereCollisionRadius=0
      boundBone {boneName=rope_02}
      boxExtents=<10 15 15>
      constraintSetup {angularLimitsMax=<7 3 7>, angularLimitsMin=<-7 -3 -7>, angularTarget=<1 0 0>}
      localJointOffset=<4 0 0>
     }]
   }
   n3 = Anim Dynamics - Bone: rope_03(ComponentPose=@n2, Alpha=1, ExternalForce="0,0,0") {
    angularDampingOverride=0.9999
    angularSpringConstant=20000
    bOverrideAngularDamping=true
    bUseGravityOverride=true
    simulationSpace=RootRelative
    alphaBoolBlend {blendOption=Linear}
    boundBone {boneName=rope_03}
    componentAppliedLinearAccClamp=<1500 1500 1500>
    componentLinearAccScale=<0.15 0.15 0.15>
    componentLinearVelScale=<0.15 0.15 0.15>
    gravityOverride {z=-980}
    physicsBodyDefinitions [{
      sphereCollisionRadius=0
      boundBone {boneName=rope_03}
      boxExtents=<10 15 15>
      constraintSetup {angularLimitsMax=<7 3 7>, angularLimitsMin=<-7 -3 -7>, angularTarget=<1 0 0>}
      localJointOffset=<4 0 0>
     }]
   }
   n4 = Anim Dynamics - Bone: rope_04(ComponentPose=@n3, Alpha=1, ExternalForce="0,0,0") {
    angularDampingOverride=0.9999
    angularSpringConstant=20000
    bOverrideAngularDamping=true
    bUseGravityOverride=true
    simulationSpace=RootRelative
    alphaBoolBlend {blendOption=Linear}
    boundBone {boneName=rope_04}
    componentAppliedLinearAccClamp=<1500 1500 1500>
    componentLinearAccScale=<0.15 0.15 0.15>
    componentLinearVelScale=<0.15 0.15 0.15>
    gravityOverride {z=-980}
    physicsBodyDefinitions [{
      sphereCollisionRadius=0
      boundBone {boneName=rope_04}
      boxExtents=<10 15 15>
      constraintSetup {angularLimitsMax=<7 3 7>, angularLimitsMin=<-7 -3 -7>, angularTarget=<1 0 0>}
      localJointOffset=<4 0 0>
     }]
   }
   n5 = Anim Dynamics - Bone: rope_05(ComponentPose=@n4, Alpha=1, ExternalForce="0,0,0") {
    angularDampingOverride=0.9999
    angularSpringConstant=20000
    bOverrideAngularDamping=true
    bUseGravityOverride=true
    simulationSpace=RootRelative
    alphaBoolBlend {blendOption=Linear}
    boundBone {boneName=rope_05}
    componentAppliedLinearAccClamp=<1500 1500 1500>
    componentLinearAccScale=<0.15 0.15 0.15>
    componentLinearVelScale=<0.15 0.15 0.15>
    gravityOverride {z=-980}
    physicsBodyDefinitions [{
      sphereCollisionRadius=0
      boundBone {boneName=rope_05}
      boxExtents=<10 15 15>
      constraintSetup {angularLimitsMax=<7 3 7>, angularLimitsMin=<-7 -3 -7>, angularTarget=<1 0 0>}
      localJointOffset=<4 0 0>
     }]
   }
   n6 = Anim Dynamics - Bone: rope_06(ComponentPose=@n5, Alpha=1, ExternalForce="0,0,0") {
    angularDampingOverride=0.9999
    angularSpringConstant=20000
    bOverrideAngularDamping=true
    bUseGravityOverride=true
    simulationSpace=RootRelative
    alphaBoolBlend {blendOption=Linear}
    boundBone {boneName=rope_06}
    componentAppliedLinearAccClamp=<1500 1500 1500>
    componentLinearAccScale=<0.15 0.15 0.15>
    componentLinearVelScale=<0.15 0.15 0.15>
    gravityOverride {z=-980}
    physicsBodyDefinitions [{
      sphereCollisionRadius=0
      boundBone {boneName=rope_06}
      boxExtents=<10 15 15>
      constraintSetup {angularLimitsMax=<7 3 7>, angularLimitsMin=<-7 -3 -7>, angularTarget=<1 0 0>}
      localJointOffset=<4 0 0>
     }]
   }
   n7 = Anim Dynamics - Bone: rope_07(ComponentPose=@n6, Alpha=1, ExternalForce="0,0,0") {
    angularDampingOverride=0.9999
    angularSpringConstant=20000
    bOverrideAngularDamping=true
    bUseGravityOverride=true
    simulationSpace=RootRelative
    alphaBoolBlend {blendOption=Linear}
    boundBone {boneName=rope_07}
    componentAppliedLinearAccClamp=<1500 1500 1500>
    componentLinearAccScale=<0.15 0.15 0.15>
    componentLinearVelScale=<0.15 0.15 0.15>
    gravityOverride {z=-980}
    physicsBodyDefinitions [{
      sphereCollisionRadius=0
      boundBone {boneName=rope_07}
      boxExtents=<10 15 15>
      constraintSetup {angularLimitsMax=<7 3 7>, angularLimitsMin=<-7 -3 -7>, angularTarget=<1 0 0>}
      localJointOffset=<4 0 0>
     }]
   }
   n8 = Anim Dynamics - Bone: rope_08(ComponentPose=@n7, Alpha=1, ExternalForce="0,0,0") {
    angularDampingOverride=0.9999
    angularSpringConstant=20000
    bOverrideAngularDamping=true
    bUseGravityOverride=true
    simulationSpace=RootRelative
    alphaBoolBlend {blendOption=Linear}
    boundBone {boneName=rope_08}
    componentAppliedLinearAccClamp=<1500 1500 1500>
    componentLinearAccScale=<0.15 0.15 0.15>
    componentLinearVelScale=<0.15 0.15 0.15>
    gravityOverride {z=-980}
    physicsBodyDefinitions [{
      sphereCollisionRadius=0
      boundBone {boneName=rope_08}
      boxExtents=<10 15 15>
      constraintSetup {angularLimitsMax=<7 3 7>, angularLimitsMin=<-7 -3 -7>, angularTarget=<1 0 0>}
      localJointOffset=<4 0 0>
     }]
   }
   n9 = Anim Dynamics - Bone: rope_09(ComponentPose=@n8, Alpha=1, ExternalForce="0,0,0") {
    angularDampingOverride=0.9999
    angularSpringConstant=20000
    bOverrideAngularDamping=true
    bUseGravityOverride=true
    simulationSpace=RootRelative
    alphaBoolBlend {blendOption=Linear}
    boundBone {boneName=rope_09}
    componentAppliedLinearAccClamp=<1500 1500 1500>
    componentLinearAccScale=<0.15 0.15 0.15>
    componentLinearVelScale=<0.15 0.15 0.15>
    gravityOverride {z=-980}
    physicsBodyDefinitions [{
      sphereCollisionRadius=0
      boundBone {boneName=rope_09}
      boxExtents=<10 15 15>
      constraintSetup {angularLimitsMax=<7 3 7>, angularLimitsMin=<-7 -3 -7>, angularTarget=<1 0 0>}
      localJointOffset=<4 0 0>
     }]
   }
   n10 = Anim Dynamics - Bone: rope_10(ComponentPose=@n9, Alpha=1, ExternalForce="0,0,0") {
    angularDampingOverride=0.9999
    angularSpringConstant=20000
    bOverrideAngularDamping=true
    bUseGravityOverride=true
    simulationSpace=RootRelative
    alphaBoolBlend {blendOption=Linear}
    boundBone {boneName=rope_10}
    componentAppliedLinearAccClamp=<1500 1500 1500>
    componentLinearAccScale=<0.15 0.15 0.15>
    componentLinearVelScale=<0.15 0.15 0.15>
    gravityOverride {z=-980}
    physicsBodyDefinitions [{
      sphereCollisionRadius=0
      boundBone {boneName=rope_10}
      boxExtents=<10 15 15>
      constraintSetup {angularLimitsMax=<7 3 7>, angularLimitsMin=<-7 -3 -7>, angularTarget=<1 0 0>}
      localJointOffset=<4 0 0>
     }]
   }
   n11 = Anim Dynamics - Bone: rope_11(ComponentPose=@n10, Alpha=1, ExternalForce="0,0,0") {
    angularDampingOverride=0.9999
    angularSpringConstant=20000
    bOverrideAngularDamping=true
    bUseGravityOverride=true
    simulationSpace=RootRelative
    alphaBoolBlend {blendOption=Linear}
    boundBone {boneName=rope_11}
    componentAppliedLinearAccClamp=<1500 1500 1500>
    componentLinearAccScale=<0.15 0.15 0.15>
    componentLinearVelScale=<0.15 0.15 0.15>
    gravityOverride {z=-980}
    physicsBodyDefinitions [{
      sphereCollisionRadius=0
      boundBone {boneName=rope_11}
      boxExtents=<10 15 15>
      constraintSetup {angularLimitsMax=<7 3 7>, angularLimitsMin=<-7 -3 -7>, angularTarget=<1 0 0>}
      localJointOffset=<4 0 0>
     }]
   }
   n12 = Anim Dynamics - Bone: rope_12(ComponentPose=@n11, Alpha=1, ExternalForce="0,0,0") {
    angularDampingOverride=0.9999
    angularSpringConstant=20000
    bOverrideAngularDamping=true
    bUseGravityOverride=true
    simulationSpace=RootRelative
    alphaBoolBlend {blendOption=Linear}
    boundBone {boneName=rope_12}
    componentAppliedLinearAccClamp=<1500 1500 1500>
    componentLinearAccScale=<0.15 0.15 0.15>
    componentLinearVelScale=<0.15 0.15 0.15>
    gravityOverride {z=-980}
    physicsBodyDefinitions [{
      sphereCollisionRadius=0
      boundBone {boneName=rope_12}
      boxExtents=<10 15 15>
      constraintSetup {angularLimitsMax=<7 3 7>, angularLimitsMin=<-7 -3 -7>, angularTarget=<1 0 0>}
      localJointOffset=<4 0 0>
     }]
   }
   n13 = Anim Dynamics - Bone: rope_13(ComponentPose=@n12, Alpha=1, ExternalForce="0,0,0") {
    angularDampingOverride=0.9999
    angularSpringConstant=20000
    bOverrideAngularDamping=true
    bUseGravityOverride=true
    simulationSpace=RootRelative
    alphaBoolBlend {blendOption=Linear}
    boundBone {boneName=rope_13}
    componentAppliedLinearAccClamp=<1500 1500 1500>
    componentLinearAccScale=<0.15 0.15 0.15>
    componentLinearVelScale=<0.15 0.15 0.15>
    gravityOverride {z=-980}
    physicsBodyDefinitions [{
      sphereCollisionRadius=0
      boundBone {boneName=rope_13}
      boxExtents=<10 15 15>
      constraintSetup {angularLimitsMax=<7 3 7>, angularLimitsMin=<-7 -3 -7>, angularTarget=<1 0 0>}
      localJointOffset=<4 0 0>
     }]
   }
   n14 = Anim Dynamics - Bone: rope_14(ComponentPose=@n13, Alpha=1, ExternalForce="0,0,0") {
    angularDampingOverride=0.9999
    angularSpringConstant=20000
    bOverrideAngularDamping=true
    bUseGravityOverride=true
    simulationSpace=RootRelative
    alphaBoolBlend {blendOption=Linear}
    boundBone {boneName=rope_14}
    componentAppliedLinearAccClamp=<1500 1500 1500>
    componentLinearAccScale=<0.15 0.15 0.15>
    componentLinearVelScale=<0.15 0.15 0.15>
    gravityOverride {z=-980}
    physicsBodyDefinitions [{
      sphereCollisionRadius=0
      boundBone {boneName=rope_14}
      boxExtents=<10 15 15>
      constraintSetup {angularLimitsMax=<7 3 7>, angularLimitsMin=<-7 -3 -7>, angularTarget=<1 0 0>}
      localJointOffset=<4 0 0>
     }]
   }
   n15 = Anim Dynamics - Bone: rope_15(ComponentPose=@n14, Alpha=1, ExternalForce="0,0,0") {
    angularDampingOverride=0.9999
    angularSpringConstant=20000
    bOverrideAngularDamping=true
    bUseGravityOverride=true
    simulationSpace=RootRelative
    alphaBoolBlend {blendOption=Linear}
    boundBone {boneName=rope_15}
    componentAppliedLinearAccClamp=<1500 1500 1500>
    componentLinearAccScale=<0.15 0.15 0.15>
    componentLinearVelScale=<0.15 0.15 0.15>
    gravityOverride {z=-980}
    physicsBodyDefinitions [{
      sphereCollisionRadius=0
      boundBone {boneName=rope_15}
      boxExtents=<10 15 15>
      constraintSetup {angularLimitsMax=<7 3 7>, angularLimitsMin=<-7 -3 -7>, angularTarget=<1 0 0>}
      localJointOffset=<4 0 0>
     }]
   }
   n16 = Anim Dynamics - Bone: rope_16(ComponentPose=@n15, Alpha=1, ExternalForce="0,0,0") {
    angularDampingOverride=0.9999
    angularSpringConstant=20000
    bOverrideAngularDamping=true
    bUseGravityOverride=true
    simulationSpace=RootRelative
    alphaBoolBlend {blendOption=Linear}
    boundBone {boneName=rope_16}
    componentAppliedLinearAccClamp=<1500 1500 1500>
    componentLinearAccScale=<0.15 0.15 0.15>
    componentLinearVelScale=<0.15 0.15 0.15>
    gravityOverride {z=-980}
    physicsBodyDefinitions [{
      sphereCollisionRadius=0
      boundBone {boneName=rope_16}
      boxExtents=<10 15 15>
      constraintSetup {angularLimitsMax=<7 3 7>, angularLimitsMin=<-7 -3 -7>, angularTarget=<1 0 0>}
      localJointOffset=<4 0 0>
     }]
   }
   n17 = Anim Dynamics - Bone: rope_17(ComponentPose=@n16, Alpha=1, ExternalForce="0,0,0") {
    angularDampingOverride=0.9999
    angularSpringConstant=20000
    bOverrideAngularDamping=true
    bUseGravityOverride=true
    simulationSpace=RootRelative
    alphaBoolBlend {blendOption=Linear}
    boundBone {boneName=rope_17}
    componentAppliedLinearAccClamp=<1500 1500 1500>
    componentLinearAccScale=<0.15 0.15 0.15>
    componentLinearVelScale=<0.15 0.15 0.15>
    gravityOverride {z=-980}
    physicsBodyDefinitions [{
      sphereCollisionRadius=0
      boundBone {boneName=rope_17}
      boxExtents=<10 15 15>
      constraintSetup {angularLimitsMax=<7 3 7>, angularLimitsMin=<-7 -3 -7>, angularTarget=<1 0 0>}
      localJointOffset=<4 0 0>
     }]
   }
   n18 = Anim Dynamics - Bone: rope_18(ComponentPose=@n17, Alpha=1, ExternalForce="0,0,0") {
    angularDampingOverride=0.9999
    angularSpringConstant=20000
    bOverrideAngularDamping=true
    bUseGravityOverride=true
    simulationSpace=RootRelative
    alphaBoolBlend {blendOption=Linear}
    boundBone {boneName=rope_18}
    componentAppliedLinearAccClamp=<1500 1500 1500>
    componentLinearAccScale=<0.15 0.15 0.15>
    componentLinearVelScale=<0.15 0.15 0.15>
    gravityOverride {z=-980}
    physicsBodyDefinitions [{
      sphereCollisionRadius=0
      boundBone {boneName=rope_18}
      boxExtents=<10 15 15>
      constraintSetup {angularLimitsMax=<7 3 7>, angularLimitsMin=<-7 -3 -7>, angularTarget=<1 0 0>}
      localJointOffset=<4 0 0>
     }]
   }
   n19 = Anim Dynamics - Bone: rope_19(ComponentPose=@n18, Alpha=1, ExternalForce="0,0,0") {
    angularDampingOverride=0.9999
    angularSpringConstant=20000
    bOverrideAngularDamping=true
    bUseGravityOverride=true
    simulationSpace=RootRelative
    alphaBoolBlend {blendOption=Linear}
    boundBone {boneName=rope_19}
    componentAppliedLinearAccClamp=<1500 1500 1500>
    componentLinearAccScale=<0.15 0.15 0.15>
    componentLinearVelScale=<0.15 0.15 0.15>
    gravityOverride {z=-980}
    physicsBodyDefinitions [{
      sphereCollisionRadius=0
      boundBone {boneName=rope_19}
      boxExtents=<10 15 15>
      constraintSetup {angularLimitsMax=<7 3 7>, angularLimitsMin=<-7 -3 -7>, angularTarget=<1 0 0>}
      localJointOffset=<4 0 0>
     }]
   }
   n20 = Anim Dynamics - Bone: rope_20(ComponentPose=@n19, Alpha=1, ExternalForce="0,0,0") {
    angularDampingOverride=0.9999
    angularSpringConstant=20000
    bOverrideAngularDamping=true
    bUseGravityOverride=true
    simulationSpace=RootRelative
    alphaBoolBlend {blendOption=Linear}
    boundBone {boneName=rope_20}
    componentAppliedLinearAccClamp=<1500 1500 1500>
    componentLinearAccScale=<0.15 0.15 0.15>
    componentLinearVelScale=<0.15 0.15 0.15>
    gravityOverride {z=-980}
    physicsBodyDefinitions [{
      sphereCollisionRadius=0
      boundBone {boneName=rope_20}
      boxExtents=<10 15 15>
      constraintSetup {angularLimitsMax=<7 3 7>, angularLimitsMin=<-7 -3 -7>, angularTarget=<1 0 0>}
      localJointOffset=<4 0 0>
     }]
   }
   n21 = Anim Dynamics - Bone: rope_21(ComponentPose=@n20, Alpha=1, ExternalForce="0,0,0") {
    angularDampingOverride=0.9999
    angularSpringConstant=20000
    bOverrideAngularDamping=true
    bUseGravityOverride=true
    simulationSpace=RootRelative
    alphaBoolBlend {blendOption=Linear}
    boundBone {boneName=rope_21}
    componentAppliedLinearAccClamp=<1500 1500 1500>
    componentLinearAccScale=<0.15 0.15 0.15>
    componentLinearVelScale=<0.15 0.15 0.15>
    gravityOverride {z=-980}
    physicsBodyDefinitions [{
      sphereCollisionRadius=0
      boundBone {boneName=rope_21}
      boxExtents=<10 15 15>
      constraintSetup {angularLimitsMax=<7 3 7>, angularLimitsMin=<-7 -3 -7>, angularTarget=<1 0 0>}
      localJointOffset=<4 0 0>
     }]
   }
   n22 = Anim Dynamics - Bone: rope_22(ComponentPose=@n21, Alpha=1, ExternalForce="0,0,0") {
    angularDampingOverride=0.9999
    angularSpringConstant=20000
    bOverrideAngularDamping=true
    bUseGravityOverride=true
    simulationSpace=RootRelative
    alphaBoolBlend {blendOption=Linear}
    boundBone {boneName=rope_22}
    componentAppliedLinearAccClamp=<1500 1500 1500>
    componentLinearAccScale=<0.15 0.15 0.15>
    componentLinearVelScale=<0.15 0.15 0.15>
    gravityOverride {z=-980}
    physicsBodyDefinitions [{
      sphereCollisionRadius=0
      boundBone {boneName=rope_22}
      boxExtents=<10 15 15>
      constraintSetup {angularLimitsMax=<7 3 7>, angularLimitsMin=<-7 -3 -7>, angularTarget=<1 0 0>}
      localJointOffset=<4 0 0>
     }]
   }
   n23 = Anim Dynamics - Bone: rope_23(ComponentPose=@n22, Alpha=1, ExternalForce="0,0,0") {
    angularDampingOverride=0.9999
    angularSpringConstant=20000
    bOverrideAngularDamping=true
    bUseGravityOverride=true
    simulationSpace=RootRelative
    alphaBoolBlend {blendOption=Linear}
    boundBone {boneName=rope_23}
    componentAppliedLinearAccClamp=<1500 1500 1500>
    componentLinearAccScale=<0.15 0.15 0.15>
    componentLinearVelScale=<0.15 0.15 0.15>
    gravityOverride {z=-980}
    physicsBodyDefinitions [{
      sphereCollisionRadius=0
      boundBone {boneName=rope_23}
      boxExtents=<10 15 15>
      constraintSetup {angularLimitsMax=<7 3 7>, angularLimitsMin=<-7 -3 -7>, angularTarget=<1 0 0>}
      localJointOffset=<4 0 0>
     }]
   }
   Anim Dynamics - Bone: rope_24(ComponentPose=@n23, Alpha=1, ExternalForce="0,0,0") {
    angularDampingOverride=0.9999
    angularSpringConstant=20000
    bOverrideAngularDamping=true
    bUseGravityOverride=true
    simulationSpace=RootRelative
    alphaBoolBlend {blendOption=Linear}
    boundBone {boneName=rope_24}
    componentAppliedLinearAccClamp=<1500 1500 1500>
    componentLinearAccScale=<0.15 0.15 0.15>
    componentLinearVelScale=<0.15 0.15 0.15>
    gravityOverride {z=-980}
    physicsBodyDefinitions [{
      sphereCollisionRadius=0
      boundBone {boneName=rope_24}
      boxExtents=<10 15 15>
      constraintSetup {angularLimitsMax=<7 3 7>, angularLimitsMin=<-7 -3 -7>, angularTarget=<1 0 0>}
      localJointOffset=<4 0 0>
     }]
   }
   AnimGraphNode_StateMachine_0 = Grapple Anim() {is=title AnimGraphNode_StateMachine, bAllowConduitEntryStates=true}
   Output Pose(Result=Dead Blending!(Source:Pose Link Structure=@AnimGraphNode_StateMachine_0) -> (Pose:Pose Link Structure) {is=title AnimGraphNode_DeadBlending, defaultBlendMode=Linear}) {layerGroup=None}
  ]}
 Grapple Anim {
  entry=Conduit
  path=AnimGraph.AnimGraphNode_StateMachine_0.Grapple Anim
  ids=id
  nodes [
   AnimStateConduitNode_0 = Conduit()
   AnimStateNode_0 = Idle()
   AnimStateNode_1 = AirAttack() {is=title AnimStateNode}
   AnimStateNode_2 = FullbodyAttack_1() {is=title AnimStateNode, bAlwaysResetOnEntry=true}
   AnimStateNode_3 = FullbodyAttack_2() {is=title AnimStateNode, bAlwaysResetOnEntry=true}
   AnimStateNode_6 = UpperBodyAttack_1() {is=title AnimStateNode}
   AnimStateNode_7 = UpperBodyAttack_2() {is=title AnimStateNode}
   AnimStateTransitionNode_0 = AnimStateTransitionNode() {crossfadeDuration=0.1, logicType=TLT_Inertialization}
   AnimStateTransitionNode_1 = AnimStateTransitionNode() {crossfadeDuration=0.1, logicType=TLT_Inertialization}
   AnimStateTransitionNode_10 = AnimStateTransitionNode() {automaticRuleTriggerTime=0, bAutomaticRuleBasedOnSequencePlayerInState=true}
   AnimStateTransitionNode_11 = AnimStateTransitionNode() {crossfadeDuration=0.1, priorityOrder=5}
   AnimStateTransitionNode_12 = AnimStateTransitionNode() {automaticRuleTriggerTime=0, bAutomaticRuleBasedOnSequencePlayerInState=true}
   AnimStateTransitionNode_13 = AnimStateTransitionNode() {crossfadeDuration=0.1, logicType=TLT_Inertialization}
   AnimStateTransitionNode_14 = AnimStateTransitionNode() {crossfadeDuration=0.1, logicType=TLT_Inertialization}
   AnimStateTransitionNode_15 = AnimStateTransitionNode() {automaticRuleTriggerTime=0, bAutomaticRuleBasedOnSequencePlayerInState=true}
   AnimStateTransitionNode_2 = AnimStateTransitionNode() {crossfadeDuration=0.3, logicType=TLT_Inertialization, priorityOrder=100}
   AnimStateTransitionNode_3 = AnimStateTransitionNode()
   AnimStateTransitionNode_4 = AnimStateTransitionNode() {logicType=TLT_Inertialization}
   AnimStateTransitionNode_5 = AnimStateTransitionNode() {automaticRuleTriggerTime=0, bAutomaticRuleBasedOnSequencePlayerInState=true}
   AnimStateTransitionNode_6 = AnimStateTransitionNode() {crossfadeDuration=0.1, logicType=TLT_Inertialization, priorityOrder=2}
   AnimStateTransitionNode_7 = AnimStateTransitionNode() {crossfadeDuration=0.1, logicType=TLT_Inertialization, priorityOrder=4}
   AnimStateTransitionNode_8 = AnimStateTransitionNode() {crossfadeDuration=0.1, logicType=TLT_Inertialization, priorityOrder=3}
   AnimStateTransitionNode_9 = AnimStateTransitionNode() {automaticRuleTriggerTime=0, bAutomaticRuleBasedOnSequencePlayerInState=true}
  ]
  links [
   {
    dst=Conduit
    rule=AnimGraph.AnimGraphNode_StateMachine_0.Grapple Anim.AnimStateTransitionNode_5.Transition
    src=AirAttack
    via=AnimStateTransitionNode_5
   }
   {
    dst=AirAttack
    rule=AnimGraph.AnimGraphNode_StateMachine_0.Grapple Anim.AnimStateTransitionNode_4.Transition
    src=Conduit
    via=AnimStateTransitionNode_4
   }
   {
    dst=FullbodyAttack_1
    rule=AnimGraph.AnimGraphNode_StateMachine_0.Grapple Anim.AnimStateTransitionNode_6.Transition
    src=Conduit
    via=AnimStateTransitionNode_6
   }
   {
    dst=FullbodyAttack_2
    rule=AnimGraph.AnimGraphNode_StateMachine_0.Grapple Anim.AnimStateTransitionNode_8.Transition
    src=Conduit
    via=AnimStateTransitionNode_8
   }
   {
    dst=Idle
    rule=AnimGraph.AnimGraphNode_StateMachine_0.Grapple Anim.AnimStateTransitionNode_2.Transition
    src=Conduit
    via=AnimStateTransitionNode_2
   }
   {
    dst=UpperBodyAttack_1
    rule=AnimGraph.AnimGraphNode_StateMachine_0.Grapple Anim.AnimStateTransitionNode_7.Transition
    src=Conduit
    via=AnimStateTransitionNode_7
   }
   {
    dst=UpperBodyAttack_2
    rule=AnimGraph.AnimGraphNode_StateMachine_0.Grapple Anim.AnimStateTransitionNode_11.Transition
    src=Conduit
    via=AnimStateTransitionNode_11
   }
   {
    dst=Conduit
    rule=AnimGraph.AnimGraphNode_StateMachine_0.Grapple Anim.AnimStateTransitionNode_9.Transition
    src=FullbodyAttack_1
    via=AnimStateTransitionNode_9
   }
   {
    dst=FullbodyAttack_2
    rule=AnimGraph.AnimGraphNode_StateMachine_0.Grapple Anim.AnimStateTransitionNode_0.Transition
    src=FullbodyAttack_1
    via=AnimStateTransitionNode_0
   }
   {
    dst=Conduit
    rule=AnimGraph.AnimGraphNode_StateMachine_0.Grapple Anim.AnimStateTransitionNode_10.Transition
    src=FullbodyAttack_2
    via=AnimStateTransitionNode_10
   }
   {
    dst=FullbodyAttack_1
    rule=AnimGraph.AnimGraphNode_StateMachine_0.Grapple Anim.AnimStateTransitionNode_1.Transition
    src=FullbodyAttack_2
    via=AnimStateTransitionNode_1
   }
   {
    dst=Conduit
    rule=AnimGraph.AnimGraphNode_StateMachine_0.Grapple Anim.AnimStateTransitionNode_3.Transition
    src=Idle
    via=AnimStateTransitionNode_3
   }
   {
    dst=Conduit
    rule=AnimGraph.AnimGraphNode_StateMachine_0.Grapple Anim.AnimStateTransitionNode_15.Transition
    src=UpperBodyAttack_1
    via=AnimStateTransitionNode_15
   }
   {
    dst=UpperBodyAttack_2
    rule=AnimGraph.AnimGraphNode_StateMachine_0.Grapple Anim.AnimStateTransitionNode_13.AnimationTransitionGraph_1
    src=UpperBodyAttack_1
    via=AnimStateTransitionNode_13
   }
   {
    dst=Conduit
    rule=AnimGraph.AnimGraphNode_StateMachine_0.Grapple Anim.AnimStateTransitionNode_12.Transition
    src=UpperBodyAttack_2
    via=AnimStateTransitionNode_12
   }
   {
    dst=UpperBodyAttack_1
    rule=AnimGraph.AnimGraphNode_StateMachine_0.Grapple Anim.AnimStateTransitionNode_14.AnimationTransitionGraph_1
    src=UpperBodyAttack_2
    via=AnimStateTransitionNode_14
   }
  ]
 }
 Conduit {path=AnimGraph.AnimGraphNode_StateMachine_0.Grapple Anim.AnimStateConduitNode_0.Conduit, nodes [Result(bCanEnterTransition=true)]}
 Idle {
  path=AnimGraph.AnimGraphNode_StateMachine_0.Grapple Anim.AnimStateNode_0.Idle
  nodes [Output Animation Pose(Result=Sequence Player 'Grappleknife_Idle_Pose_WEP_Knife_Sharp'() {bLoopAnimation=false, sequence=/GrappleWeapon/Items/Animations/Grappleknife_Idle_Pose_WEP_Knife_Sharp})]
 }
 AirAttack {
  path=AnimGraph.AnimGraphNode_StateMachine_0.Grapple Anim.AnimStateNode_1.AirAttack
  nodes [Output Animation Pose(Result=Sequence Player 'GrappleKnife_inAirAttack_WEP_knife'() {bLoopAnimation=false, sequence=/GrappleWeapon/Items/Animations/WEP/GrappleKnife_inAirAttack_WEP_knife})]
 }
 FullbodyAttack_1 {
  path=AnimGraph.AnimGraphNode_StateMachine_0.Grapple Anim.AnimStateNode_2.FullbodyAttack_1
  nodes [Output Animation Pose(Result=Sequence Player 'GrappleKnife_GroundAttack_01_male_WEP_knife'() {bLoopAnimation=false, sequence=/GrappleWeapon/Items/Animations/WEP/GrappleKnife_GroundAttack_01_male_WEP_knife})]
 }
 FullbodyAttack_2 {
  path=AnimGraph.AnimGraphNode_StateMachine_0.Grapple Anim.AnimStateNode_3.FullbodyAttack_2
  nodes [Output Animation Pose(Result=Sequence Player 'GrappleKnife_GroundAttack_02_male_WEP_knife'() {bLoopAnimation=false, sequence=/GrappleWeapon/Items/Animations/WEP/GrappleKnife_GroundAttack_02_male_WEP_knife})]
 }
 UpperBodyAttack_1 {
  path=AnimGraph.AnimGraphNode_StateMachine_0.Grapple Anim.AnimStateNode_6.UpperBodyAttack_1
  nodes [Output Animation Pose(Result=Sequence Player 'GrappleKnife_GroundAttack_01_UpperbodyOnly_WEP_knife'() {bLoopAnimation=false, sequence=/GrappleWeapon/Items/Animations/WEP/GrappleKnife_GroundAttack_01_UpperbodyOnly_WEP_knife})]
 }
 UpperBodyAttack_2 {
  path=AnimGraph.AnimGraphNode_StateMachine_0.Grapple Anim.AnimStateNode_7.UpperBodyAttack_2
  nodes [Output Animation Pose(Result=Sequence Player 'GrappleKnife_GroundAttack_02_UpperbodyOnly_WEP_knife'() {bLoopAnimation=false, sequence=/GrappleWeapon/Items/Animations/WEP/GrappleKnife_GroundAttack_02_UpperbodyOnly_WEP_knife})]
 }
 Transition {
  path=AnimGraph.AnimGraphNode_StateMachine_0.Grapple Anim.AnimStateTransitionNode_0.Transition
  nodes [Result(bCanEnterTransition=Get MantisAttack_2() -> (MantisAttack_2:bool))]
 }
 Transition {
  path=AnimGraph.AnimGraphNode_StateMachine_0.Grapple Anim.AnimStateTransitionNode_1.Transition
  nodes [Result(bCanEnterTransition=Get MantisAttack_1() -> (MantisAttack_1:bool))]
 }
 Transition {path=AnimGraph.AnimGraphNode_StateMachine_0.Grapple Anim.AnimStateTransitionNode_10.Transition, nodes [Result(bCanEnterTransition=true)]}
 Transition {
  path=AnimGraph.AnimGraphNode_StateMachine_0.Grapple Anim.AnimStateTransitionNode_11.Transition
  nodes [Result(bCanEnterTransition=Get MantisAttack_4() -> (MantisAttack_4:bool))]
 }
 Transition {path=AnimGraph.AnimGraphNode_StateMachine_0.Grapple Anim.AnimStateTransitionNode_12.Transition, nodes [Result()]}
 AnimationTransitionGraph_1 {
  path=AnimGraph.AnimGraphNode_StateMachine_0.Grapple Anim.AnimStateTransitionNode_13.AnimationTransitionGraph_1
  nodes [Result(bCanEnterTransition=Get MantisAttack_4())]
 }
 AnimationTransitionGraph_1 {
  path=AnimGraph.AnimGraphNode_StateMachine_0.Grapple Anim.AnimStateTransitionNode_14.AnimationTransitionGraph_1
  nodes [Result(bCanEnterTransition=Get MantisAttack_3() -> (MantisAttack_3:bool))]
 }
 Transition {path=AnimGraph.AnimGraphNode_StateMachine_0.Grapple Anim.AnimStateTransitionNode_15.Transition, nodes [Result()]}
 Transition {path=AnimGraph.AnimGraphNode_StateMachine_0.Grapple Anim.AnimStateTransitionNode_2.Transition, nodes [Result(bCanEnterTransition=true)]}
 Transition {path=AnimGraph.AnimGraphNode_StateMachine_0.Grapple Anim.AnimStateTransitionNode_3.Transition, nodes [Result(bCanEnterTransition=true)]}
 Transition {
  path=AnimGraph.AnimGraphNode_StateMachine_0.Grapple Anim.AnimStateTransitionNode_4.Transition
  nodes [Result(bCanEnterTransition=Get MantisAttackAir() -> (MantisAttackAir:bool))]
 }
 Transition {path=AnimGraph.AnimGraphNode_StateMachine_0.Grapple Anim.AnimStateTransitionNode_5.Transition, nodes [Result()]}
 Transition {
  path=AnimGraph.AnimGraphNode_StateMachine_0.Grapple Anim.AnimStateTransitionNode_6.Transition
  nodes [Result(bCanEnterTransition=Get MantisAttack_1())]
 }
 Transition {
  path=AnimGraph.AnimGraphNode_StateMachine_0.Grapple Anim.AnimStateTransitionNode_7.Transition
  nodes [Result(bCanEnterTransition=Get MantisAttack_3())]
 }
 Transition {
  path=AnimGraph.AnimGraphNode_StateMachine_0.Grapple Anim.AnimStateTransitionNode_8.Transition
  nodes [Result(bCanEnterTransition=Get MantisAttack_2())]
 }
 Transition {path=AnimGraph.AnimGraphNode_StateMachine_0.Grapple Anim.AnimStateTransitionNode_9.Transition, nodes [Result(bCanEnterTransition=true)]}
 EventGraph {nodes [
   n1 = Event OnMantisTechniqueStarted!() -> (OutputDelegate, then, TechniqueIndex:int=0, TechniqueName:Name=None) {is=title K2Node_Event}
   > Set IsAirAttack(IsAirAttack:bool="Equal (Name)"(A=@n1.TechniqueName, B=InAirAttack)) -> (Output_Get:bool)
   n2 = Event OnMantisTechniqueEnded!() -> (OutputDelegate, then, TechniqueIndex:int=0, TechniqueName:Name=None) {is=title K2Node_Event}
   > Set IsAirAttack(IsAirAttack=Select(Option 0:bool=false, Option 1:bool=false, Index="Equal (Name)"(A=@n2.TechniqueName, B=InAirAttack)) -> (ReturnValue:bool))
   PrintString!(execute, InString="To String (Name)"(InName=@n1.TechniqueName), bPrintToScreen, bPrintToLog, TextColor, Duration, Key) -> (then)
  ]}
]
variables [IsAirAttack {type=boolean}]
